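# Niagara swing trail for the ClumsyChewOffhand pickaxe, combining anim-trail ribbons with leaf sprites.
# Two ribbon pairs of differing width, green or white, span two sockets, with User.VariantID gating each pair's trail count through MaskInt.
# Leaf sprites spawn by distance moved at a socket, inherit velocity, drift on curl noise and random gravity, and lerp green to pink by VariantID.
# Time-of-day inputs adjust brightness or opacity on most emitters.
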NiagaraSystem /BRCosmetics/Weapons/FORT_Melee/Pickaxe_ClumsyChewOffhand/FX/NS_Pickaxe_ClumsyChewOffhand_Trail
dependencies {
 usedDataInterfaces [&NiagaraDataInterfaceCurve, &NiagaraDataInterfaceSocketReader]
 usedDynamicInputs []
 usedModules [
  &self:SystemSpawnScript.SystemScriptSource.SystemScriptGraph.NiagaraNodeAssignment_1.SetVariables_#1
  &SystemState
 ]
}
emitters {
 NE_Leafs {
  emitterUpdateScript [
   &EmitterState(Life Cycle Mode=::NewEnumerator1, Loop Behavior=::NewEnumerator1, Loop Delay=0.05, Loop Duration=0.3, UseLoopDelay=true)
   &SpawnPerUnit(
    Max Movement Threshold=2500
    Module.Use Max Movement Threshold=true
    Module.Use Movement Tolerance=false
    Spawn Spacing:NiagaraFloat=&RandomRangeFloat(Minimum=35, Maximum=70)
    Use Spawn Probability=false
   )
  ]
  particleSpawnScript [
   &InitializeParticle(
    Color=&LerpLinearColors(StartColor=<rgba 0.012 0.1719 0 1>, EndColor=<rgba 1 0.1779 0.1779 1>, LerpFactor=$User.VariantID)
    Lifetime Max=0.5
    Lifetime Min=0.15
    Lifetime Mode=::NewEnumerator1
    Sprite Rotation Mode=::NewEnumerator2
    Sprite Size Mode=::NewEnumerator1
    Uniform Sprite Size Max=16
    Uniform Sprite Size Min=8
    UsePositionOffset=false
    Write Material Random=true
   )
   &SocketLocation(
    Interpolate Socket Sampling=true
    SamplePosOffset=false
    Socket Index=1
    Socket Reader=$System.Socket Reader
    Socket Sampling Mode=::NewEnumerator2
    VelocitySpeedLimit=false
   )
   &InheritVelocity(
    Inherited Velocity Speed Limit=20
    Speed Limit=true
    UseSpeedThreshold=false
    Velocity Source=$Particles.SkeletalMeshLocation.SampledVelocity
   )
   &ShapeLocation(MaskSpawnGroup=false, Override Seed=false, Sphere Radius=25, UseOwnerScale=true)
  ]
  particleUpdateScript [
   &ParticleState()
   &ScaleSpriteSize(Uniform Curve Sprite Scale=DataInterface() {propertyValues {Curve {defaultValue=0, keys [{tangentMode=RCTM_User, time=0, value=1}, {tangentMode=RCTM_User, time=1, value=0}]}}})
   &GravityForce(Gravity=&RandomRangeVector(Minimum=<0 0 300>, Maximum=<0 0 -300>))
   &CurlNoiseForce(Bias Noise=false, Noise Frequency=0.1, Noise Strength=500, Pan Noise=false)
   &SolveForcesAndVelocity(Clamp Velocity=false, Limit Acceleration=false)
   &ScaleColor(
    Scale Alpha=&FloatFromCurve(FloatCurve=DataInterface() {propertyValues {Curve {keys [<at 0 1>,<at 1 0>]}}})
    Scale RGB=&VectorFromFloat(Value=&FNTimeOfDayScalar(Night Scale=0.8, Morning Scale=0.7, Evening Scale=0.7))
    ScaleA=true
    ScaleRGB=true
   )
   &SpriteRotationRate(Rotation Rate=&RandomRangeFloat(Minimum=-45, Maximum=45))
   &DynamicMaterialParameters(
    Index 0 Param 1=&FNTimeOfDayScalar(Night Scale=0.15, Morning Scale=1, Evening Scale=1)
    Module.Index0Float0Write=true
    Module.Index0Float1Write=true
    Module.Index0Float2Write=true
    Module.Index0Float3Write=true
   )
  ]
  rendererClasses [&NiagaraSpriteRendererProperties]
  renderers [{rendererClass=&NiagaraSpriteRendererProperties}]
 }
 NE_Trail_01 {
  emitterSpawnScript [&FN_RibbonIDOffsetUtility()]
  emitterUpdateScript [
   &EmitterState()
   &FN_SpawnAnimTrail(Number of Anim Trails:NiagaraInt32=/Niagara/DynamicInputs/Helpers/MaskInt(
     Mask Source:NiagaraInt32=1
     Mask Parameter:NiagaraInt32=&MakeIntFromFloat(Float=$User.VariantID)
     Mask:NiagaraInt32=1
     Masked Int:NiagaraInt32=0
     Passthrough Int:NiagaraInt32=1
    ))
  ]
  particleSpawnScript [
   &FN_InitializeAnimTrailBoneIndecies(First Bone or Socket Index=1, Second Bone or Socket Index=0)
   &FN_InitializeAnimTrail(AnimTrailWidth=1.8, Color=<rgba 0.0897 0.172 0.0597 1>, Lifetime=0.2)
  ]
  particleUpdateScript [
   &SolveForcesAndVelocity(Clamp Velocity=false, Limit Acceleration=false)
   &FN_UpdateAnimTrail() {module=FN_UpdateAnimTrail001}
   &ScaleColor(
    Scale Alpha=&FloatFromCurve(FloatCurve=DataInterface() {propertyValues {Curve {keys [
         {arriveTangent=-4.827, interpMode=RCIM_Cubic, leaveTangent=-4.827, tangentMode=RCTM_User, time=0, value=1}
         {arriveTangent=-0.2078, interpMode=RCIM_Cubic, leaveTangent=-0.2078, tangentMode=RCTM_User, time=0.4, value=0}
        ]}}})
    Scale RGB=&VectorFromFloat(Value=&FNTimeOfDayScalar(Night Scale=0.5, Morning Scale=1, Evening Scale=1))
    ScaleA=true
    ScaleRGB=true
   )
  ]
  rendererClasses [&NiagaraRibbonRendererProperties]
  renderers [{rendererClass=&NiagaraRibbonRendererProperties}]
 }
 NE_Trail_01_VAR {
  emitterSpawnScript [&FN_RibbonIDOffsetUtility()]
  emitterUpdateScript [
   &EmitterState()
   &FN_SpawnAnimTrail(Number of Anim Trails:NiagaraInt32=&MaskInt(
     Mask Source:NiagaraInt32=1
     Mask Parameter:NiagaraInt32=&MakeIntFromFloat(Float=$User.VariantID)
     Mask:NiagaraInt32=1
     Masked Int:NiagaraInt32=1
     Passthrough Int:NiagaraInt32=0
    ))
  ]
  particleSpawnScript [
   &FN_InitializeAnimTrailBoneIndecies(First Bone or Socket Index=1, Second Bone or Socket Index=0)
   &FN_InitializeAnimTrail(AnimTrailWidth=1.8, Color=<rgba 1 1 1 0.251>, Lifetime=0.2)
  ]
  particleUpdateScript [
   &SolveForcesAndVelocity(Clamp Velocity=false, Limit Acceleration=false)
   &FN_UpdateAnimTrail() {module=FN_UpdateAnimTrail001}
   &ScaleColor(
    Scale Alpha=&FloatFromCurve(FloatCurve=DataInterface() {propertyValues {Curve {keys [
         {arriveTangent=-4.827, interpMode=RCIM_Cubic, leaveTangent=-4.827, tangentMode=RCTM_User, time=0, value=1}
         {arriveTangent=-0.2078, interpMode=RCIM_Cubic, leaveTangent=-0.2078, tangentMode=RCTM_User, time=0.4, value=0}
        ]}}})
    ScaleA=true
    ScaleRGB=true
   )
  ]
  rendererClasses [&NiagaraRibbonRendererProperties]
  renderers [{rendererClass=&NiagaraRibbonRendererProperties}]
 }
 NE_Trail_02 {
  emitterSpawnScript [&FN_RibbonIDOffsetUtility()]
  emitterUpdateScript [
   &EmitterState()
   &FN_SpawnAnimTrail(Number of Anim Trails:NiagaraInt32=&MaskInt(
     Mask Source:NiagaraInt32=1
     Mask Parameter:NiagaraInt32=&MakeIntFromFloat(Float=$User.VariantID)
     Mask:NiagaraInt32=1
     Masked Int:NiagaraInt32=0
     Passthrough Int:NiagaraInt32=1
    ))
  ]
  particleSpawnScript [
   &FN_InitializeAnimTrailBoneIndecies(First Bone or Socket Index=1, Second Bone or Socket Index=0)
   &FN_InitializeAnimTrail(AnimTrailWidth=1.1, Color=<rgba 0.0897 0.172 0.0597 1>, Lifetime=0.2)
  ]
  particleUpdateScript [
   &SolveForcesAndVelocity(Clamp Velocity=false, Limit Acceleration=false)
   &FN_UpdateAnimTrail() {module=FN_UpdateAnimTrail001}
   &ScaleColor(
    Scale Alpha=&FloatFromCurve(FloatCurve=DataInterface() {propertyValues {Curve {keys [
         {arriveTangent=-4.827, interpMode=RCIM_Cubic, leaveTangent=-4.827, tangentMode=RCTM_User, time=0, value=1}
         {arriveTangent=-0.2078, interpMode=RCIM_Cubic, leaveTangent=-0.2078, tangentMode=RCTM_User, time=0.4, value=0}
        ]}}})
    Scale RGB=&VectorFromFloat(Value=&FNTimeOfDayScalar(Night Scale=0.5, Morning Scale=1, Evening Scale=1))
    ScaleA=true
    ScaleRGB=true
   )
  ]
  rendererClasses [&NiagaraRibbonRendererProperties]
  renderers [{rendererClass=&NiagaraRibbonRendererProperties}]
 }
 NE_Trail_02_VAR {
  emitterSpawnScript [&FN_RibbonIDOffsetUtility()]
  emitterUpdateScript [
   &EmitterState()
   &FN_SpawnAnimTrail(Number of Anim Trails:NiagaraInt32=&MaskInt(
     Mask Source:NiagaraInt32=1
     Mask Parameter:NiagaraInt32=&MakeIntFromFloat(Float=$User.VariantID)
     Mask:NiagaraInt32=1
     Masked Int:NiagaraInt32=1
     Passthrough Int:NiagaraInt32=0
    ))
  ]
  particleSpawnScript [
   &FN_InitializeAnimTrailBoneIndecies(First Bone or Socket Index=1, Second Bone or Socket Index=0)
   &FN_InitializeAnimTrail(
    AnimTrailWidth=1.1
    Color=&FNTimeOfDayColor(Night Color=<rgba 1 1 1 0.15>, Morning Color=<rgba 1 1 1 0.5>, Day Color=<rgba 1 1 1 1>, Evening Color=<rgba 1 1 1 0.5>)
    Lifetime=0.2
   )
  ]
  particleUpdateScript [
   &SolveForcesAndVelocity(Clamp Velocity=false, Limit Acceleration=false)
   &FN_UpdateAnimTrail() {module=FN_UpdateAnimTrail001}
   &ScaleColor(
    Scale Alpha=&FloatFromCurve(FloatCurve=DataInterface() {propertyValues {Curve {keys [
         {arriveTangent=-4.827, interpMode=RCIM_Cubic, leaveTangent=-4.827, tangentMode=RCTM_User, time=0, value=1}
         {arriveTangent=-0.2078, interpMode=RCIM_Cubic, leaveTangent=-0.2078, tangentMode=RCTM_User, time=0.4, value=0}
        ]}}})
    ScaleA=true
    ScaleRGB=true
   )
  ]
  rendererClasses [&NiagaraRibbonRendererProperties]
  renderers [{rendererClass=&NiagaraRibbonRendererProperties}]
 }
}
summary {
 systemName=NS_Pickaxe_ClumsyChewOffhand_Trail
 emitters [
  {emitterName=NE_Trail_01, rendererClasses [&NiagaraRibbonRendererProperties]}
  {emitterName=NE_Trail_02, rendererClasses [&NiagaraRibbonRendererProperties]}
  {emitterName=NE_Trail_01_VAR, rendererClasses [&NiagaraRibbonRendererProperties]}
  {emitterName=NE_Leafs, rendererClasses [&NiagaraSpriteRendererProperties]}
  {emitterName=NE_Trail_02_VAR, rendererClasses [&NiagaraRibbonRendererProperties]}
 ]
 userVariables [
  User.bIsFrontEnd {type=NiagaraBool}
  User.bIsFrontEndPreview {type=NiagaraBool}
  User.PhysicalSurface {type=NiagaraInt32}
  User.RibbonIDOffset {type=NiagaraInt32}
  User.VariantID {type=NiagaraFloat, defaultValue=0, description=""}
 ]
}
